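; Anim notify that spawns the Jim gauntlet "JustGlints" flare particle on the animated mesh's left hand (hand_l).
; It casts the mesh owner to PlayerPawn_Athena and checks it for six SpecialActor.Ashton.Object color tags (B, G, O, P, R, Y).
; Each tag result enables or disables the matching color-named emitter in the spawned effect.
; It also prints all six tag states to screen and log as debug output.
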(fn Received_Notify
  (params (MeshComp :type Skeletal%20Mesh%20Component%20Object%20Reference) (Animation :type Anim%20Sequence%20Base%20Object%20Reference) (EventReference :type Anim%20Notify%20Event%20Reference%20Structure%20%28by%20ref%29))
  (bind _asplayer_pawn_athena (Utilities|Casting|CastToPlayerPawn_Athena (Components|GetOwner MeshComp)))
  (bind _returnvalue (Effects|SpawnEmitterAttached "/Figment_Martin/Athena/Items/Gameplay/BackPacks/CarminePack/FX/P_Jim_Spawning_Gauntlet_Flare_JustGlints.P_Jim_Spawning_Gauntlet_Flare_JustGlints" MeshComp "hand_l"))
  (Development|PrintString (Utilities|String|Append "B - " (Utilities|String|ToString(Boolean) (GameplayTags|HasMatchingGameplayTag _asplayer_pawn_athena "(TagName=\"SpecialActor.Ashton.Object.B\")")) "; G - " (Utilities|String|ToString(Boolean) (GameplayTags|HasMatchingGameplayTag _asplayer_pawn_athena "(TagName=\"SpecialActor.Ashton.Object.G\")")) "; O - " (Utilities|String|ToString(Boolean) (GameplayTags|HasMatchingGameplayTag _asplayer_pawn_athena "(TagName=\"SpecialActor.Ashton.Object.O\")")) "; P - " (Utilities|String|ToString(Boolean) (GameplayTags|HasMatchingGameplayTag _asplayer_pawn_athena "(TagName=\"SpecialActor.Ashton.Object.P\")")) "; R - " (Utilities|String|ToString(Boolean) (GameplayTags|HasMatchingGameplayTag _asplayer_pawn_athena "(TagName=\"SpecialActor.Ashton.Object.R\")")) "; Y - " (Utilities|String|ToString(Boolean) (GameplayTags|HasMatchingGameplayTag _asplayer_pawn_athena "(TagName=\"SpecialActor.Ashton.Object.Y\")"))) true true "(R=0.000000,G=0.660000,B=1.000000,A=1.000000)" 1e+01)
  (Effects|Components|ParticleSystem|SetEmitterEnable _returnvalue "Blue" (GameplayTags|HasMatchingGameplayTag _asplayer_pawn_athena "(TagName=\"SpecialActor.Ashton.Object.B\")"))
  (Effects|Components|ParticleSystem|SetEmitterEnable _returnvalue "Green" (GameplayTags|HasMatchingGameplayTag _asplayer_pawn_athena "(TagName=\"SpecialActor.Ashton.Object.G\")"))
  (Effects|Components|ParticleSystem|SetEmitterEnable _returnvalue "Orange" (GameplayTags|HasMatchingGameplayTag _asplayer_pawn_athena "(TagName=\"SpecialActor.Ashton.Object.O\")"))
  (Effects|Components|ParticleSystem|SetEmitterEnable _returnvalue "Purple" (GameplayTags|HasMatchingGameplayTag _asplayer_pawn_athena "(TagName=\"SpecialActor.Ashton.Object.P\")"))
  (Effects|Components|ParticleSystem|SetEmitterEnable _returnvalue "Red" (GameplayTags|HasMatchingGameplayTag _asplayer_pawn_athena "(TagName=\"SpecialActor.Ashton.Object.R\")"))
  (Effects|Components|ParticleSystem|SetEmitterEnable _returnvalue "Yellow" (GameplayTags|HasMatchingGameplayTag _asplayer_pawn_athena "(TagName=\"SpecialActor.Ashton.Object.Y\")"))
  (return false))
