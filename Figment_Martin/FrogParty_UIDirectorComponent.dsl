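; Frog Party UI director component that adds the mode's status HUD scene and toggles tagged HUD elements for the villain role (Chitauri, per node comments).
; It follows the local player's view target and waits for the villain gameplay tag on that pawn's PlayerState.
; Tag added unhides the elements; tag removed, noted as promotion to Thanos, hides them again.
; A delta gate applies visibility only when the hidden state changes.
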(event EventOnReadyToShowUI
  (CallFunction|AddDefaultHUDScene)
  (CallFunction|RegisterforEvents))

(event Custom|OnViewTargetChanged
  (params (PlayerController :type Fort%20Player%20Controller%20Object%20Reference) (OldViewTarget :type Actor%20Object%20Reference) (NewViewTarget :type Actor%20Object%20Reference))
  ; Get the new view target whenever it changes so we can wait for them to be promoted to Thanos
  (bind _aspawn (Utilities|Casting|CastToPawn NewViewTarget))
  (CallFunction|SetViewTarget _aspawn))

(event Custom|SetViewTarget
  (params (NewPawn :type Pawn%20Object%20Reference))
  (bind _playerstate (Variables|/Script/Engine.Pawn|GetPlayerState NewPawn))
  (Utilities|IsValid NewPawn
    (:Is%20Valid
      (Utilities|IsValid _playerstate
        (:Is%20Valid
          (Ability|Async|EndAction (Variables|Default|GetLeaderTagAddedAction))
          ; Check if client has Chitauri Tag
          (bind _asynctaskproxy (Ability|Tasks|WaitGameplayTagAddToActor _playerstate (Variables|Default|GetVillainTag))
            (:then
              (Variables|Default|SetLeaderTagAddedAction _asynctaskproxy))
            (:Added
              ; Unhide the UI for Chitauri and wait for when they stop being Chitari( only happens when they become Thanos)
              (CallFunction|LeaderHideHUD false)
              (bind _asynctaskproxy_2 (Ability|Tasks|WaitGameplayTagRemoveFromActor _playerstate (Variables|Default|GetVillainTag))
                (:then
                  (Variables|Default|SetLeaderTagRemovedAction _asynctaskproxy_2))
                (:Removed
                  (CallFunction|LeaderHideHUD true))))))))))

(event Custom|Leader_HideHUD
  (params (Hide :type Boolean))
  (bind _updated_variable (Utilities|FlowControl|DeltaGateWildcard (Variables|Default|GetIsHUDHidden) Hide)
    (:Success
      (bind _asfort_player_controller (Utilities|Casting|CastToFortPlayerController (CallFunction|GetOwningLocalPlayerController)))
      (HUD|SetHUDElementVisibilitywithContext _asfort_player_controller (Variables|Default|GetHUDElement%20Tags) _updated_variable "None" ""))))

(fn AddDefaultHUDScene
  (CallFunction|AddScene (Variables|Default|GetFrogPartStatusScene)))

(fn RegisterForEvents
  (bind _asfort_player_controller (Utilities|Casting|CastToFortPlayerController (CallFunction|GetOwningLocalPlayerController)))
  (bind _output_get (LocalVariables|SetFortPlayerController _asfort_player_controller))
  (EventDispatchers|BindEventtoOnViewTargetChanged _output_get (EventDispatchers|CreateEvent :!event OnViewTargetChanged))
  (bind _aspawn (Utilities|Casting|CastToPawn (Pawn|GetViewTarget (LocalVariables|GetFortPlayerController))))
  (CallFunction|SetViewTarget _aspawn))
